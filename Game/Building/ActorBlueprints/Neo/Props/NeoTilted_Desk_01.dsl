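; Neo tilted desk prop Blueprint, a building actor with a spot light.
; Its construction script runs the parent script, then checks the platform name.
; On Switch it turns off shadow casting on the spot light.
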(fn ConstructionScript
  (|CallParentFunction)
  (if (Utilities|String|EqualExactly(String) (Game|GetPlatformName) "Switch")
    (Rendering|Components|Light|SetCastShadows (Variables|Default|GetSpotLight) false)))
