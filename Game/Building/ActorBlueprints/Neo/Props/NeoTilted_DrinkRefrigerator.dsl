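; NeoTilted_DrinkRefrigerator is a drink refrigerator prop Blueprint in the Neo building prop set.
; On BeginPlay it calls the parent, then hides its point light, without propagating to children, when running on Nintendo Switch.
; Its construction script only calls the parent, so the remaining behavior comes from the parent class.
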(event EventBeginPlay
  (|CallParentFunction)
  (if (TouchPlatform|IsNintendoSwitch)
    (Rendering|SetVisibility (Variables|Default|GetPointLight) false false)))

(fn ConstructionScript
  (|CallParentFunction))
