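; Fire hydrant prop that, once damaged, becomes a water geyser launching players and hoverboards upward.
; Damage sets the "Exploded" custom state, whose change swaps meshes, starts water effects and audio, and enables jumping.
; Sphere and cylinder overlaps call LaunchActor, which launches pawns only on capsule overlap (passed by cylinder alone), offsetting vertical velocity.
; A gameplay tag blocks relaunches, an effect is applied, and next falling damage is ignored.
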(event OnComponentBeginOverlap(Touch_Sphere)
  (params (OverlappedComponent :type Primitive%20Component%20Object%20Reference) (OtherActor :type Actor%20Object%20Reference) (OtherComp :type Primitive%20Component%20Object%20Reference) (OtherBodyIndex :type Integer) (bFromSweep :type Boolean) (SweepResult :type Hit%20Result%20Structure%20%28by%20ref%29))
  (if (Variables|FireHydrant__2|GetCan_Jump)
    (CallFunction|LaunchActor OtherActor)))

(event OnComponentBeginOverlap(Touch_Cylinder)
  (params (OverlappedComponent :type Primitive%20Component%20Object%20Reference) (OtherActor :type Actor%20Object%20Reference) (OtherComp :type Primitive%20Component%20Object%20Reference) (OtherBodyIndex :type Integer) (bFromSweep :type Boolean) (SweepResult :type Hit%20Result%20Structure%20%28by%20ref%29))
  (if (Variables|FireHydrant__2|GetCan_Jump)
    (CallFunction|LaunchActor OtherActor OtherComp)))

(event EventCustomStateChanged
  (CallFunction|HandleExplodedState))

(event EventOnDamageServer
  (params (Damage :type Float%20%28single-precision%29) (DamageTags :type Gameplay%20Tag%20Container%20Structure%20%28by%20ref%29) (Momentum :type Vector :default "0, 0, 0") (HitInfo :type Hit%20Result%20Structure%20%28by%20ref%29) (InstigatedBy :type Controller%20Object%20Reference) (DamageCauser :type Actor%20Object%20Reference) (EffectContext :type Gameplay%20Effect%20Context%20Handle%20Structure))
  (Building|SetCustomState "Exploded"))

(event EventOnDeathPlayEffects
  (params (Damage :type Float%20%28single-precision%29) (DamageTags :type Gameplay%20Tag%20Container%20Structure%20%28by%20ref%29) (Momentum :type Vector :default "0, 0, 0") (HitInfo :type Hit%20Result%20Structure%20%28by%20ref%29) (InstigatedBy :type Fort%20Pawn%20Object%20Reference) (DamageCauser :type Actor%20Object%20Reference) (EffectContext :type Gameplay%20Effect%20Context%20Handle%20Structure))
  (Audio|Components|Audio|FadeOut (Variables|FireHydrant__2|GetWater_Loop_Audio) 0.2 0.0 "Linear" :!class /Script/Engine.AudioComponent)
  (Audio|PlaySoundatLocation "/Game/Sounds/Fort_GamePlay_Sounds/Bouncers/FireHydrant_WaterStream_Stop_Cue.FireHydrant_WaterStream_Stop_Cue" (Transformation|GetWorldLocation (Variables|FireHydrant__2|GetWater_Loop_Audio))))

(fn HandleExplodedState
  (bind _returnvalue (Building|GetBuildingMeshComponent))
  (bind _water_loop_audio (Variables|FireHydrant__2|GetWater_Loop_Audio))
  (bind _broken_hydrant (Variables|FireHydrant__2|GetBroken_Hydrant))
  (bind _water_effect (Variables|FireHydrant__2|GetWater_Effect))
  (switch string (Building|GetCustomState)
    (:Default
      (Variables|FireHydrant__2|SetCan_Jump false)
      (Rendering|SetVisibility _returnvalue true false)
      (Rendering|SetVisibility _broken_hydrant false false)
      (Components|Activation|Deactivate _water_effect))
    (:Exploded
      (Variables|FireHydrant__2|SetCan_Jump true)
      (Rendering|SetVisibility _returnvalue false false)
      (Collision|SetCollisionResponsetoChannel _returnvalue "ECC_GameTraceChannel7" "ECR_Ignore")
      (Rendering|SetVisibility _broken_hydrant true false)
      (Components|Activation|Activate _water_effect false)
      (Audio|Components|Audio|FadeIn _water_loop_audio 0.0 1.0 0.0 "Linear" :!class /Script/Engine.AudioComponent)
      (Components|Activation|Activate _water_loop_audio false))))

(fn LaunchActor
  (params (ActorToLaunch :type Actor%20Object%20Reference) (ComponentTriggeringLaunch :type Primitive%20Component%20Object%20Reference))
  (bind _water_effect (Variables|FireHydrant__2|GetWater_Effect))
  (bind _playertolaunch (LocalVariables|GetPlayerToLaunch))
  (bind _water_effect_2 (Variables|FireHydrant__2|GetWater_Effect))
  (bind _asfort_player_pawn (Utilities|Casting|CastToFortPlayerPawn ActorToLaunch)
    (:then
      (if (not (GameplayTags|HasMatchingGameplayTag (Ability|GetAbilitySystemComponent _asfort_player_pawn) "(TagName=\"Gameplay.Action.Player.Launched\")"))
        (LocalVariables|SetPlayerToLaunch _asfort_player_pawn)
        (if (Utilities|Equal(Object) (Variables|/Script/Engine.Character|GetCapsuleComponent _asfort_player_pawn) ComponentTriggeringLaunch)
          (Components|Activation|Deactivate _water_effect)
          (Components|Activation|Activate _water_effect false)
          (Character|LaunchCharacter _playertolaunch (Math|Vector|MakeVector 0.0 0.0 (+ (* (.z (Transformation|GetVelocity (LocalVariables|GetPlayerToLaunch))) -1) 2000)) false false)
          (GameplayEffects|ApplyGameplayEffectToSelf :self (Ability|GetAbilitySystemComponent _playertolaunch) :GameplayEffectClass (Variables|Default|GetGE_PlayerLaunched) :Level 0.0)
          (Variables|/Script/FortniteGame.FortPawn|SetbIgnoreNextFallingDamage true _playertolaunch)
          (Components|Activation|Activate (Variables|FireHydrant__2|GetLaunch_Audio) false)
          (bind _asfort_player_controller (Utilities|Casting|CastToFortPlayerController (Pawn|GetController _playertolaunch)))
          (Game|Feedback|LocalPlayForceFeedback _asfort_player_controller (Variables|Default|GetLaunchForceFeedback) false "Hydrant"))))
    (:CastFailed
      (bind _success (Vehicles|LaunchHoverboard ActorToLaunch "0, 0,2000.000000" "Hydrant"))
      (if _success
        (Components|Activation|Deactivate _water_effect_2)
        (Components|Activation|Activate _water_effect_2 false)
        (Components|Activation|Activate (Variables|FireHydrant__2|GetLaunch_Audio) false)))))

(fn ConstructionScript)
