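; Snow cone food truck prop for Neo Tilted that plays a music loop while it stands.
; On BeginPlay, where cosmetic events can execute, it plays the audio component and registers it with the MusicEventSubsystem as a POI event.
; When death effects play, it fades the loop out briefly and removes the event from the subsystem.
; Both events call their parent versions, and the audio work only runs on a valid component.
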(event EventBeginPlay
  (bind _musicloop (Variables|Default|GetMusicLoop))
  (|CallParentFunction)
  ;[1 1 1] Play music loop
  (Networking|CanExecuteCosmeticEvents
    (:True
      (Utilities|IsValid _musicloop
        (:Is%20Valid
          (Audio|Components|Audio|Play _musicloop 0.0)
          (Audio|AddEventbyComponent (GameInstanceSubsystems|GetMusicEventSubsystem) _musicloop "(TagName=\"MusicEventSubsystem.Event.POI\")" "(GameplayTags=((TagName=\"MusicEventSubsystem\")))" 0 ""))))))

(event EventOnDeathPlayEffects
  (params (Damage :type Float%20%28single-precision%29) (DamageTags :type Gameplay%20Tag%20Container%20Structure%20%28by%20ref%29) (Momentum :type Vector :default "0, 0, 0") (HitInfo :type Hit%20Result%20Structure%20%28by%20ref%29) (InstigatedBy :type Fort%20Pawn%20Object%20Reference) (DamageCauser :type Actor%20Object%20Reference) (EffectContext :type Gameplay%20Effect%20Context%20Handle%20Structure))
  (|CallParentFunction Damage DamageTags Momentum HitInfo InstigatedBy DamageCauser EffectContext)
  ;[1 1 1] End music loop
  (bind _musicloop (Variables|Default|GetMusicLoop)
    (:then
      (Audio|Components|Audio|FadeOut _musicloop 0.25 0.0 "Linear" :!class /Script/Engine.AudioComponent)
      (Audio|RemoveEvent (GameInstanceSubsystems|GetMusicEventSubsystem) _musicloop))
    (:else)))

(fn ConstructionScript
  (|CallParentFunction))
