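; Blueprint for a tilted spotlight building prop in the Neo set, configured almost entirely in its construction script.
; After running the parent construction script, it disables shadow casting on its point light when the platform is Switch.
; It then sets a uniform actor scale of one plus a hundredth of the Rotation variable, so that value drives size despite its name.
; BeginPlay is empty.
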(event EventBeginPlay)

(fn ConstructionScript
  (|CallParentFunction)
  (Utilities|FlowControl|Sequence
    (:then_0
      (if (Utilities|String|EqualExactly(String) (Game|GetPlatformName) "Switch")
        (Rendering|Components|Light|SetCastShadows (Variables|Default|GetPointLight) false)))
    (:then_1
      (Transformation|SetActorScale3D (Math|Vector|MakeVector (+ (* (Variables|Default|GetRotation) 0.01) 1.0) (+ (* (Variables|Default|GetRotation) 0.01) 1.0) (+ (* (Variables|Default|GetRotation) 0.01) 1.0))))))
